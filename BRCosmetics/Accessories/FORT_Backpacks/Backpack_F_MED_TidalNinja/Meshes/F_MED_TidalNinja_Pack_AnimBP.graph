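# Animation Blueprint for the TidalNinja backpack cosmetic, posing the pack against the wearer's back.
# Each update caches world transforms of the attach parent's pelvis and spine sockets; the AnimGraph has world-space replace nodes for the matching backpack bones.
# Rotation constraints blend spine rotations onto the dyn_pack bones, followed by fixed additive offsets.
# Bool-driven offsets adjust for crouch, skydive float, dive and the large "henchman tough" body, detected by mesh-part name.
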AnimBlueprint /BRCosmetics/Accessories/FORT_Backpacks/Backpack_F_MED_TidalNinja/Meshes/F_MED_TidalNinja_Pack_AnimBP
target_skeleton=&/F_MED_TidalNinja_Pack_Skeleton
graphs [
 AnimGraph {nodes [
   n1 = "Transform (Modify) Bone - Bone: backpack_pelvis"!(ComponentPose=Local To Component(LocalPose=Use cached pose 'input pose'()), Alpha=1, Translation="0,0,0", Rotation="0,0,0") -> (Pose) {
    rotationMode=BMM_Replace
    rotationSpace=BCS_WorldSpace
    scaleSpace=BCS_BoneSpace
    translationMode=BMM_Replace
    translationSpace=BCS_WorldSpace
    alphaBoolBlend {blendOption=Linear}
    alphaScaleBiasClamp {inRange {max=0, min=-40}, outRange {max=0, min=1}}
    boneToModify {boneName=backpack_pelvis}
   }
   n2 = "Transform (Modify) Bone - Bone: backpack_spine_1"!(ComponentPose=@n1, Alpha=1, Translation="0,0,0", Rotation="0,0,0") -> (Pose) {
    rotationMode=BMM_Replace
    rotationSpace=BCS_WorldSpace
    scaleSpace=BCS_BoneSpace
    translationMode=BMM_Replace
    translationSpace=BCS_WorldSpace
    alphaBoolBlend {blendOption=Linear}
    alphaScaleBiasClamp {inRange {max=0, min=-40}, outRange {max=0, min=1}}
    boneToModify {boneName=backpack_spine_1}
   }
   n3 = "Transform (Modify) Bone - Bone: backpack_spine_2"!(ComponentPose=@n2, Alpha=1, Translation="0,0,0", Rotation="0,0,0") -> (Pose) {
    rotationMode=BMM_Replace
    rotationSpace=BCS_WorldSpace
    scaleSpace=BCS_BoneSpace
    translationMode=BMM_Replace
    translationSpace=BCS_WorldSpace
    alphaBoolBlend {blendOption=Linear}
    alphaScaleBiasClamp {inRange {max=0, min=-40}, outRange {max=0, min=1}}
    boneToModify {boneName=backpack_spine_2}
   }
   n4 = "Transform (Modify) Bone - Bone: backpack_spine_3"!(ComponentPose=@n3, Alpha=1, Translation="0,0,0", Rotation="0,0,0") -> (Pose) {
    rotationMode=BMM_Replace
    rotationSpace=BCS_WorldSpace
    scaleSpace=BCS_BoneSpace
    translationMode=BMM_Replace
    translationSpace=BCS_WorldSpace
    alphaBoolBlend {blendOption=Linear}
    alphaScaleBiasClamp {inRange {max=0, min=-40}, outRange {max=0, min=1}}
    boneToModify {boneName=backpack_spine_3}
   }
   n5 = "Transform (Modify) Bone - Bone: backpack_spine_4"!(ComponentPose=@n4, Alpha=1, Translation="0,0,0", Rotation="0,0,0") -> (Pose) {
    rotationMode=BMM_Replace
    rotationSpace=BCS_WorldSpace
    scaleSpace=BCS_BoneSpace
    translationMode=BMM_Replace
    translationSpace=BCS_WorldSpace
    alphaBoolBlend {blendOption=Linear}
    alphaScaleBiasClamp {inRange {max=0, min=-40}, outRange {max=0, min=1}}
    boneToModify {boneName=backpack_spine_4}
   }
   n6 = "Transform (Modify) Bone - Bone: backpack_spine_5"!(ComponentPose=@n5, Alpha=1, Translation="0,0,0", Rotation="0,0,0") -> (Pose) {
    rotationMode=BMM_Replace
    rotationSpace=BCS_WorldSpace
    scaleSpace=BCS_BoneSpace
    translationMode=BMM_Replace
    translationSpace=BCS_WorldSpace
    alphaBoolBlend {blendOption=Linear}
    alphaScaleBiasClamp {inRange {max=0, min=-40}, outRange {max=0, min=1}}
    boneToModify {boneName=backpack_spine_5}
   }
   n7 = "Constraint - dyn_pack_00 (RR)"!(ComponentPose=@n6, Alpha=1, ConstraintWeights_0=1, ConstraintWeights_1:float=0) -> (Pose) {
    alphaBoolBlend {blendOption=Linear}
    boneToModify {boneName=dyn_pack_00}
    constraintSetup [{transformType=Rotation, targetBone {boneName=backpack_spine_5}}, {transformType=Rotation, targetBone {boneName=backpack_spine_4}}]
    constraintWeights [1,1]
   }
   n8 = "Constraint - dyn_pack_01 (RR)"!(ComponentPose=@n7, Alpha=1, ConstraintWeights_0=0.5, ConstraintWeights_1:float=0.5) -> (Pose) {
    alphaBoolBlend {blendOption=Linear}
    boneToModify {boneName=dyn_pack_01}
    constraintSetup [{transformType=Rotation, targetBone {boneName=backpack_spine_4}}, {transformType=Rotation, targetBone {boneName=backpack_spine_5}}]
    constraintWeights [1,1]
   }
   n9 = "Constraint - dyn_pack_02 (RR)"!(ComponentPose=@n8, Alpha=1, ConstraintWeights_0=0.4, ConstraintWeights_1:float=0.6) -> (Pose) {
    alphaBoolBlend {blendOption=Linear}
    boneToModify {boneName=dyn_pack_02}
    constraintSetup [{transformType=Rotation, targetBone {boneName=backpack_spine_4}}, {transformType=Rotation, targetBone {boneName=backpack_spine_5}}]
    constraintWeights [1,1]
   }
   n10 = "Constraint - dyn_pack_03 (RR)"!(ComponentPose=@n9, Alpha=1, ConstraintWeights_0=0.3, ConstraintWeights_1:float=0.7) -> (Pose) {
    alphaBoolBlend {blendOption=Linear}
    boneToModify {boneName=dyn_pack_03}
    constraintSetup [{transformType=Rotation, targetBone {boneName=backpack_spine_3}}, {transformType=Rotation, targetBone {boneName=backpack_spine_4}}]
    constraintWeights [1,1]
   }
   n11 = "Constraint - dyn_pack_04 (RR)"!(ComponentPose=@n10, Alpha=1, ConstraintWeights_0=0.7, ConstraintWeights_1:float=0.3) -> (Pose) {
    alphaBoolBlend {blendOption=Linear}
    boneToModify {boneName=dyn_pack_04}
    constraintSetup [{transformType=Rotation, targetBone {boneName=backpack_spine_5}}, {transformType=Rotation, targetBone {boneName=backpack_spine_4}}]
    constraintWeights [1,1]
   }
   n12 = "Constraint - dyn_pack_05 (R)"(ComponentPose=@n11, Alpha=1, ConstraintWeights_0=1) {
    alphaBoolBlend {blendOption=Linear}
    boneToModify {boneName=dyn_pack_05}
    constraintSetup [{transformType=Rotation, targetBone {boneName=backpack_spine_2}}]
    constraintWeights [1]
   }
   n13 = "Transform (Modify) Bone - Bone: dyn_pack_00"(ComponentPose=@n12, Alpha=1, Translation="0,0,0.5", Rotation="7,0,0", Scale="1,1,1") {
    rotationMode=BMM_Additive
    rotationSpace=BCS_BoneSpace
    translationMode=BMM_Additive
    translationSpace=BCS_BoneSpace
    alphaBoolBlend {blendOption=Linear}
    boneToModify {boneName=dyn_pack_00}
    rotation {pitch=7}
    translation {z=0.5}
   }
   n14 = "Transform (Modify) Bone - Bone: dyn_pack_02"(ComponentPose=@n13, Alpha=1, Translation="0,0,0", Rotation="6,0,0", Scale="1,1,1") {
    rotationMode=BMM_Additive
    rotationSpace=BCS_BoneSpace
    translationMode=BMM_Additive
    translationSpace=BCS_BoneSpace
    alphaBoolBlend {blendOption=Linear}
    boneToModify {boneName=dyn_pack_02}
    rotation {pitch=6}
   }
   n15 = "Transform (Modify) Bone - Bone: dyn_pack_04"(ComponentPose=@n14, Alpha=1, Translation="0,0,0.5", Rotation="-12,0,0", Scale="1,1,1") {
    rotationMode=BMM_Additive
    rotationSpace=BCS_BoneSpace
    translationMode=BMM_Additive
    translationSpace=BCS_BoneSpace
    alphaBoolBlend {blendOption=Linear}
    boneToModify {boneName=dyn_pack_04}
    rotation {pitch=-12}
    translation {z=0.5}
   }
   n16 = Get bIsCrouching()
   n17 = "Transform (Modify) Bone - Bone: dyn_pack_03"!(ComponentPose=@n15, bAlphaBoolEnabled:bool=@n16, Translation="0,0,0", Rotation="23,0,0", Scale="1,1,1") -> (Pose) {
    alphaInputType=Bool
    rotationMode=BMM_Additive
    rotationSpace=BCS_BoneSpace
    translationMode=BMM_Additive
    translationSpace=BCS_BoneSpace
    alphaBoolBlend {blendInTime=0.2, blendOption=Linear, blendOutTime=0.2}
    boneToModify {boneName=dyn_pack_03}
    rotation {pitch=23}
   }
   n18 = "Transform (Modify) Bone - Bone: dyn_pack_04"!(ComponentPose=@n17, bAlphaBoolEnabled:bool=@n16, Translation="0,0,-1", Rotation="-30,0.003,0.006", Scale="1,1,1") -> (Pose) {
    alphaInputType=Bool
    rotationMode=BMM_Additive
    rotationSpace=BCS_BoneSpace
    translationMode=BMM_Additive
    translationSpace=BCS_BoneSpace
    alphaBoolBlend {blendInTime=0.2, blendOption=Linear, blendOutTime=0.2}
    boneToModify {boneName=dyn_pack_04}
    rotation=<rot -30 0.003 0.006>
    translation {z=-1}
   }
   n19 = Get bIsSkydiveFloating()
   n20 = "Transform (Modify) Bone - Bone: dyn_pack_00"!(ComponentPose=@n18, bAlphaBoolEnabled:bool=@n19, Translation="0,0,0.5", Rotation="0,0,0", Scale="1,1,1") -> (Pose) {
    alphaInputType=Bool
    rotationMode=BMM_Additive
    rotationSpace=BCS_BoneSpace
    translationMode=BMM_Additive
    translationSpace=BCS_BoneSpace
    alphaBoolBlend {blendInTime=0.2, blendOption=Linear, blendOutTime=0.2}
    boneToModify {boneName=dyn_pack_00}
    translation {z=0.5}
   }
   n21 = "Transform (Modify) Bone - Bone: dyn_pack_01"!(ComponentPose=@n20, bAlphaBoolEnabled:bool=@n19, Translation="0,0,0", Rotation="-10,0,0", Scale="1,1,1") -> (Pose) {
    alphaInputType=Bool
    rotationMode=BMM_Additive
    rotationSpace=BCS_BoneSpace
    translationMode=BMM_Additive
    translationSpace=BCS_BoneSpace
    alphaBoolBlend {blendInTime=0.2, blendOption=Linear, blendOutTime=0.2}
    boneToModify {boneName=dyn_pack_01}
    rotation {pitch=-10}
   }
   n22 = "Transform (Modify) Bone - Bone: dyn_pack_03"(ComponentPose=@n21, bAlphaBoolEnabled=@n19, Translation="0,0,0", Rotation="13,0,0", Scale="1,1,1") {
    alphaInputType=Bool
    rotationMode=BMM_Additive
    rotationSpace=BCS_BoneSpace
    translationMode=BMM_Additive
    translationSpace=BCS_BoneSpace
    alphaBoolBlend {blendInTime=0.2, blendOption=Linear, blendOutTime=0.2}
    boneToModify {boneName=dyn_pack_03}
    rotation {pitch=13}
   }
   n23 = "Transform (Modify) Bone - Bone: dyn_pack_03"(ComponentPose=@n22, bAlphaBoolEnabled=Get bIsSkydiveDiveMode(), Translation="0,0,0", Rotation="10,0,0", Scale="1,1,1") {
    alphaInputType=Bool
    rotationMode=BMM_Additive
    rotationSpace=BCS_BoneSpace
    translationMode=BMM_Additive
    translationSpace=BCS_BoneSpace
    alphaBoolBlend {blendInTime=0.2, blendOption=Linear, blendOutTime=0.2}
    boneToModify {boneName=dyn_pack_03}
    rotation {pitch=10}
   }
   n24 = "Transform (Modify) Bone - Bone: dyn_pack_00"(
    ComponentPose=@n23
    bAlphaBoolEnabled=Get is Henchman Tough() -> (is Henchman Tough:bool)
    Translation="0,-1.3,0"
    Rotation="0,0,8"
    Scale="1,1,1"
   ) {
    alphaInputType=Bool
    rotationMode=BMM_Additive
    rotationSpace=BCS_ParentBoneSpace
    translationMode=BMM_Additive
    translationSpace=BCS_ParentBoneSpace
    alphaBoolBlend {blendOption=Linear}
    boneToModify {boneName=dyn_pack_00}
    rotation {roll=8}
    translation {y=-1.3}
   }
   Output Pose(Result=Component To Local(ComponentPose=@n24)) {layerGroup=None}
   Save cached pose 'input pose'(Pose=Input Pose()) {cacheName=input pose}
  ]}
 EventGraph {nodes [
   Event BlueprintUpdateAnimation()
   n1 = GetPartSkeletalMeshComponent()
   > Is Valid(InputObject=@n1)
   n2 = GetAttachParent!(self:Scene Component Object Reference=@n1) -> (ReturnValue:Scene Component Object Reference)
   > Is Valid(InputObject=@n2)
   n3 = BreakTransform(InTransform=GetSocketTransform(self=@n2, InSocketName=spine_01, TransformSpace=RTS_World))
   > Set Spine 01 T(Spine 01 T:Vector=@n3) -> (Output_Get:Vector)
   > Set Spine 01 R(Spine 01 R:Rotator=@n3.Rotation) -> (Output_Get:Rotator)
   n4 = BreakTransform(InTransform=GetSocketTransform(self=@n2, InSocketName=spine_02, TransformSpace=RTS_World))
   > Set spine_02_T(spine_02_T:Vector=@n4) -> (Output_Get:Vector)
   > Set spine_02_R(spine_02_R:Rotator=@n4.Rotation) -> (Output_Get:Rotator)
   n5 = BreakTransform(InTransform=GetSocketTransform(self=@n2, InSocketName=spine_03, TransformSpace=RTS_World))
   > Set spine_03_T(spine_03_T:Vector=@n5) -> (Output_Get:Vector)
   > Set spine_03_R(spine_03_R:Rotator=@n5.Rotation) -> (Output_Get:Rotator)
   n6 = BreakTransform(InTransform=GetSocketTransform(self=@n2, InSocketName=spine_04, TransformSpace=RTS_World))
   > Set spine_04_T(spine_04_T:Vector=@n6) -> (Output_Get:Vector)
   > Set spine_04_R(spine_04_R:Rotator=@n6.Rotation) -> (Output_Get:Rotator)
   n7 = BreakTransform(InTransform=GetSocketTransform(self=@n2, InSocketName=spine_05, TransformSpace=RTS_World))
   > Set spine_05_T(spine_05_T:Vector=@n7) -> (Output_Get:Vector)
   > Set spine_05_R(spine_05_R:Rotator=@n7.Rotation) -> (Output_Get:Rotator)
   n8 = BreakTransform(InTransform=GetSocketTransform(self=@n2, InSocketName=pelvis, TransformSpace=RTS_World))
   > Set pelvis T(pelvis T:Vector=@n8) -> (Output_Get:Vector)
   > Set pelvis R(pelvis R:Rotator=@n8.Rotation) -> (Output_Get:Rotator)
   n9 = Event OnCharacterCustomizationLoaded()
   n10 = Get Array Item with EFort Custom Part Type(Array=@n9.CustomMeshParts)
   > Set is Henchman Tough(is Henchman Tough:bool=Contains(SearchIn=GetDisplayName(Object=@n10), Substring=LRG_henchman_tough)) -> (Output_Get:bool)
   TryGetPawnOwner()
  ]}
]
variables [
 backpack_rotate {type=Rotator}
 backpack_translation {type=Vector}
 is Henchman Tough {type=boolean}
 pelvis R {type=Rotator}
 pelvis T {type=Vector}
 Spine 01 R {type=Rotator}
 Spine 01 T {type=Vector}
 spine_02_R {type=Rotator}
 spine_02_T {type=Vector}
 spine_03_R {type=Rotator}
 spine_03_T {type=Vector}
 spine_04_R {type=Rotator}
 spine_04_T {type=Vector}
 spine_05_R {type=Rotator}
 spine_05_T {type=Vector}
]
